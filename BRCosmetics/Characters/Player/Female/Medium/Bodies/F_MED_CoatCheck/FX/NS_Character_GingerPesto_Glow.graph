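# Niagara glow effect for the GingerPesto character cosmetic, placing two glows at opposite offsets on the character's skeletal mesh.
# Glow_Plane_Stacked builds each glow from five bone-oriented sprites, fanned outward with decreasing size and alpha to fake volume.
# GlowsAndLights adds a point light at each side.
# Both emitters share a system glow color and loop particle lifetime, so curves pulse size, brightness and light radius without respawning.
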NiagaraSystem /BRCosmetics/Characters/Player/Female/Medium/Bodies/F_MED_CoatCheck/FX/NS_Character_GingerPesto_Glow
dependencies {
 usedDataInterfaces [
  &NiagaraDataInterfaceArrayFloat
  &NiagaraDataInterfaceArrayInt32
  &NiagaraDataInterfaceArrayFloat3
  &NiagaraDataInterfaceCurve
  &NiagaraDataInterfaceMeshRendererInfo
  &NiagaraDataInterfaceSkeletalMesh
 ]
 usedDynamicInputs [
  &SimulationPosition
  &Add_Float
  &Multiply_FloatByInt
  &Multiply_Vector
 ]
 usedModules [
  &self:SystemSpawnScript.SystemScriptSource.SystemScriptGraph.NiagaraNodeAssignment_9.SetVariables_#1
  &self:SystemSpawnScript.SystemScriptSource.SystemScriptGraph.NiagaraNodeAssignment_0.SetVariables_#2
  &SystemState
 ]
}
emitters {
 Glow_Plane_Stacked {
  emitterUpdateScript [
   &EmitterState(
    Inactive Response=::NewEnumerator1
    Life Cycle Mode=::NewEnumerator1
    Loop Behavior=::NewEnumerator1
    Loop Duration Mode=::NewEnumerator1
    UseLoopDelay=false
   )
   &SpawnBurst_Instantaneous(Spawn Count=10, Use Spawn Probability=false, UseLoopCountLimit=false)
  ]
  particleSpawnScript [
   &FN_SetSpawnGroup(Normalize Spawn Group=true, Number of Spawn Groups=2, Particles Per Group=5, Set Ribbon ID and Ribbon Link Order=true)
   &InitializeParticle(
    Color=$System.Glow Color
    Lifetime=5
    Sprite Size=&Multiply_Vector2D_ByFloat(
     Float=&MaskFloatBySpawnGroup(Masked Float=0.9, Passthrough Float=1.8)
     Vector2D=&Multiply_Vector2D_ByFloat(
      Float=&OneMinusFloat(Float=&Power(a=$Particles.Normalized Spawn Group, b=1))
      Vector2D=<10 10>
     )
    )
    Sprite Size Mode=::NewEnumerator4
    UsePositionOffset=false
    Write Material Random=true
   )
   &CameraOffset(Camera Offset Amount=1)
  ]
  particleUpdateScript [
   &ParticleState(
    Kill Particles When Lifetime Has Elapsed=false
    Let Infinitely Lived Particles Die When Emitter Deactivates=true
    Loop Particles Lifetime=true
   )
   &ScaleSpriteSize(Uniform Curve Sprite Scale=DataInterface() {propertyValues {Curve {keys [
        {interpMode=RCIM_Cubic, tangentMode=RCTM_User, time=0, value=0.9}
        {interpMode=RCIM_Cubic, tangentMode=RCTM_User, time=0.4, value=1}
        {tangentMode=RCTM_User, time=1, value=0.9}
       ]}}})
   &SkeletalMeshLocation(
    Bone / Socket Index=&SelectIntFromArray(
     Direct Array Index=&SpawnGroup()
     Direct Array Mode=::NewEnumerator1
     Int Selection Array=DataInterface() {propertyValues {IntData [0,0]}}
    )
    Bone Sampling Mode=::NewEnumerator4
    MaskSpawnGroup=false
    Sampled Position Offset=&SelectVectorFromArray(
     Direct Array Index=&SpawnGroup()
     Direct Array Mode=::NewEnumerator1
     Vector Selection Array=DataInterface() {propertyValues {FloatData [<-1.7 10 0.5>, <1 -8.5 0.1>]}}
    )
    Skeletal Mesh=$System.Skeletal Mesh
   )
   &ScaleColor(
    Scale Alpha=&Subtract_Float(A=0.25, B=&Multiply_Float(A=$Particles.Normalized Spawn Group, B=0.25))
    Scale RGB=&VectorFromFloat(Value=&FloatFromCurve(
      FloatCurve=DataInterface() {propertyValues {Curve {keys [
          {interpMode=RCIM_Cubic, tangentMode=RCTM_User, time=0, value=0.6}
          {interpMode=RCIM_Cubic, tangentMode=RCTM_User, time=0.4, value=1}
          {tangentMode=RCTM_User, time=1, value=0.6}
         ]}}}
      Scale Curve=0.75
     ))
    ScaleA=true
    ScaleRGB=true
   )
   /SetVariables(Particles.Position=&AddVectorToPosition(
     Position=$Particles.Position
     Vector=&RotateVectorByQuaternion(
      Vector To Rotate=&MakeVector(
       X=0
       Y=&Multiply_Float(A=&MaskFloatBySpawnGroup(Masked Float=3, Passthrough Float=-3), B=$Particles.Normalized Spawn Group)
       Z=0
      )
      Quaternion=$Particles.SkeletalMeshLocation.SampledRotation
     )
    )) {isSetParameters=true}
   &SpriteFacingAndAlignment(
    Sprite Alignment=&RotateVectorByQuaternion(Vector To Rotate=<1 0 0>, Quaternion=$Particles.SkeletalMeshLocation.SampledRotation)
    Sprite Facing=&RotateVectorByQuaternion(
     Vector To Rotate=&MaskVectorBySpawnGroup(Masked Vector=<-0.1 1 0>, Passthrough Vector=<0.2 1 0>)
     Quaternion=$Particles.SkeletalMeshLocation.SampledRotation
    )
    WriteAlignment=true
    WriteFacing=true
   )
  ]
  rendererClasses [&NiagaraSpriteRendererProperties]
  renderers [{rendererClass=&NiagaraSpriteRendererProperties}]
 }
 GlowsAndLights {
  emitterUpdateScript [
   &EmitterState(
    Inactive Response=::NewEnumerator1
    Life Cycle Mode=::NewEnumerator1
    Loop Behavior=::NewEnumerator1
    Loop Duration Mode=::NewEnumerator1
    UseLoopDelay=false
   )
   &SpawnBurst_Instantaneous(Spawn Count=2, Use Spawn Probability=false, UseLoopCountLimit=true)
  ]
  particleSpawnScript [
   &PartitionParticles(ID to Partition:NiagaraInt32=&ReturnExecIndex())
   &InitializeParticle(Color Mode=::NewEnumerator0, Lifetime=5, UsePositionOffset=false, Write Material Random=true)
   /SetVariables(
    Particles.GlowsVisibilityTag:NiagaraInt32=&MergeHelper_IntSwitch(
     0=0
     1=0
     2=0
     3=0
     4=0
     5=&MakeCustomIntFromBool(Bool=$User.bShowBackFX, True INT=0, False INT=1)
     6=&MakeCustomIntFromBool(Bool=$User.bShowBackFX, True INT=0, False INT=1)
     7=0
     8=0
     Index=$Particles.PartitionParticles.Partition
     Num=::NewEnumerator7
    )
    Particles.VisibilityTag=&MakeIntFromFloat(Float=&MakeFloatFromInt(Int=&MergeHelper_IntSwitch(0=0, 1=0, 2=1, 3=1, 4=1, 5=1, 6=1, 7=1, 8=1, Index=$Particles.PartitionParticles.Partition, Num=::NewEnumerator7)))
   ) {isSetParameters=true}
  ]
  particleUpdateScript [
   &ParticleState(
    Kill Particles When Lifetime Has Elapsed=false
    Let Infinitely Lived Particles Die When Emitter Deactivates=true
    Loop Particles Lifetime=true
   )
   &SkeletalMeshLocation(
    Bone / Socket Index=&SelectIntFromArray(
     Direct Array Index=$Particles.PartitionParticles.Partition
     Direct Array Mode=::NewEnumerator1
     Int Selection Array=DataInterface() {propertyValues {IntData [0,0]}}
    )
    Bone Sampling Mode=::NewEnumerator4
    MaskSpawnGroup=false
    Sampled Position Offset=&SelectVectorFromArray(
     Direct Array Index=$Particles.PartitionParticles.Partition
     Direct Array Mode=::NewEnumerator1
     Vector Selection Array=DataInterface() {propertyValues {FloatData [<-1.5 12 0>, <-1 -12 0>]}}
    )
    Skeletal Mesh=$System.Skeletal Mesh
   )
   /SetVariables(
    Particles.LightColor=&MultiplyLinearColor(
     Linear Color=$System.Glow Color
     Scale Factor=&MakeFloat4FromVectorAndFloat(
      "Vector (RGB)"=&VectorFromFloat(Value=&Multiply_Float(
        A=&FloatFromCurve(FloatCurve=DataInterface() {propertyValues {Curve {keys [
             {interpMode=RCIM_Cubic, tangentMode=RCTM_User, time=0, value=0.75}
             {interpMode=RCIM_Cubic, tangentMode=RCTM_User, time=0.4, value=1}
             {tangentMode=RCTM_User, time=1, value=0.75}
            ]}}})
        B=&SelectFloatFromArray(Direct Array Index=$Particles.PartitionParticles.Partition, Float Selection Array=DataInterface() {propertyValues {FloatData [2.5,2.5]}})
       ))
      "Float (Alpha)"=1
     )
    )
    Particles.LightRadius=&FloatFromCurve(
     FloatCurve=DataInterface() {propertyValues {Curve {keys [
         {interpMode=RCIM_Cubic, tangentMode=RCTM_User, time=0, value=0.8}
         {interpMode=RCIM_Cubic, tangentMode=RCTM_User, time=0.4, value=1}
         {tangentMode=RCTM_User, time=1, value=0.8}
        ]}}}
     Scale Curve=20
    )
   ) {isSetParameters=true}
  ]
  rendererClasses [&NiagaraLightRendererProperties]
  renderers [{rendererClass=&NiagaraLightRendererProperties, rendererIndex=0}]
 }
}
summary {
 systemName=NS_Character_GingerPesto_Glow
 emitters [
  {emitterName=GlowsAndLights, rendererClasses [&NiagaraLightRendererProperties]}
  {emitterName=Glow_Plane_Stacked, rendererClasses [&NiagaraSpriteRendererProperties]}
 ]
 userVariables [
  User.bIsFrontEnd {type=NiagaraBool}
  User.bIsFrontEndPreview {type=NiagaraBool}
  User.bShowBackFX {type=NiagaraBool}
 ]
}
