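# Physics asset for the F_MED_CoatCheck female body, setting up physics for its dyn_skirt bones.
# Eight skirt chains around the body each have a base body angularly locked to the pelvis, followed by two capsule segments joined by constraints with per-joint swing and twist angles.
# Wider-angle constraints cross-link some neighboring lower segments.
# Capsules on the pelvis, both thighs and right calf provide body shapes; the left calf has none.
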PhysicsAsset /BRCosmetics/Characters/Player/Female/Medium/Bodies/F_MED_CoatCheck/Meshes/F_MED_CoatCheck_Physics
bodies {
 calf_r [{shapeType=Capsule, length=42.5, radius=6.5, shapeName=calf_r_capsule, center=<28.1891 2.6223 0>, rotation=<rot -90 180 174.9998>}]
 dyn_skirt_back_1 [{shapeType=Capsule, length=10.5, radius=5.5, shapeName=dyn_skirt_back_1_capsule, center=<8.7552 0 -4.1617>, rotation=<rot 0.0002 0 90>}]
 dyn_skirt_back_2 [{shapeType=Capsule, length=16, radius=6.5, shapeName=dyn_skirt_back_2_capsule, center=<8.7552 0 -4.1291>, rotation=<rot 0.0002 0 90>}]
 dyn_skirt_back_base [{shapeType=Capsule, length=0, radius=4, shapeName=dyn_skirt_back_base_capsule}]
 dyn_skirt_back_middle_1_base_l [{shapeType=Capsule, length=0, radius=4, shapeName=dyn_skirt_back_middle_1_base_l_capsule}]
 dyn_skirt_back_middle_1_l [{
   shapeType=Capsule
   length=14
   radius=7
   shapeName=dyn_skirt_back_middle_1_l_capsule
   center=<13.1396 0.4305 -1.7502>
   rotation=<rot 0.001 -0.0004 109.9999>
  }]
 dyn_skirt_back_middle_1_r [{
   shapeType=Capsule
   length=15
   radius=4.5
   shapeName=dyn_skirt_back_middle_1_r_capsule
   center=<-12.9442 -1.9478 0.8564>
   rotation=<rot 0.0009 -0.0003 -85.0001>
  }]
 dyn_skirt_back_middle_2_l [{
   shapeType=Capsule
   length=22
   radius=7
   shapeName=dyn_skirt_back_middle_2_l_capsule
   center=<10.5435 0.2976 -1.0323>
   rotation=<rot -6.7167 18.8813 108.882>
  }]
 dyn_skirt_back_middle_2_r [
  {
   shapeType=Capsule
   length=11
   radius=4.5
   shapeName=dyn_skirt_back_middle_2_r_capsule
   center=<-9.0838 6.1309 -0.0371>
   rotation=<rot 0.0008 -0.0003 -65.0001>
  }
  {
   shapeType=Capsule
   length=12.5
   radius=4.6
   shapeName=dyn_skirt_back_middle_2_r_capsule
   center=<-8.3937 -3.6173 0.3095>
   rotation=<rot 0.0006 -0.0002 75>
  }
 ]
 dyn_skirt_back_middle_base_r [{shapeType=Capsule, length=0, radius=4, shapeName=dyn_skirt_back_middle_base_r_capsule}]
 dyn_skirt_front_1 [{
   shapeType=Capsule
   length=18.75
   radius=5.25
   shapeName=dyn_skirt_front_1_capsule
   center=<12.5579 -0.1226 -1.0248>
   rotation=<rot -0.0003 0 -85>
  }]
 dyn_skirt_front_2 [{
   shapeType=Capsule
   length=15.25
   radius=5.75
   shapeName=dyn_skirt_front_2_capsule
   center=<6.0735 2.052 -4.5298>
   rotation=<rot -0.0006 20 -95.0002>
  }]
 dyn_skirt_front_base [{shapeType=Capsule, length=0, radius=4, shapeName=dyn_skirt_front_base_capsule}]
 dyn_skirt_middle_1_base_l [{shapeType=Capsule, length=0, radius=4, shapeName=dyn_skirt_middle_1_base_l_capsule}]
 dyn_skirt_middle_1_l [
  {
   shapeType=Capsule
   length=23
   radius=5.5
   shapeName=dyn_skirt_middle_1_l_capsule
   center=<22.1398 10.4339 6.0463>
   rotation=<rot -0.0015 -0.0004 39.9998>
  }
  {
   shapeType=Capsule
   length=16
   radius=5
   shapeName=dyn_skirt_middle_1_l_capsule
   center=<21.4223 -5.4692 -3.0389>
   rotation=<rot -0.0008 -0.0002 -95.0001>
  }
 ]
 dyn_skirt_middle_1_r [{
   shapeType=Capsule
   length=15.75
   radius=5.75
   shapeName=dyn_skirt_middle_1_r_capsule
   center=<-13.782 0 3.521>
   rotation=<rot -0.0002 -0.0001 90>
  }]
 dyn_skirt_middle_2_l [
  {
   shapeType=Capsule
   length=18.5
   radius=5.5
   shapeName=dyn_skirt_middle_2_l_capsule
   center=<13.188 7.5776 2.5039>
   rotation=<rot -20.4692 -151.2647 119.4517>
  }
  {
   shapeType=Capsule
   length=14
   radius=5.5
   shapeName=dyn_skirt_middle_2_l_capsule
   center=<11.6804 -5.1561 -3.8139>
   rotation=<rot 5.0269 104.3857 91.7393>
  }
 ]
 dyn_skirt_middle_2_r [{
   shapeType=Capsule
   length=15
   radius=6.5
   shapeName=dyn_skirt_middle_2_r_capsule
   center=<-15.4961 2.4716 1.8391>
   rotation=<rot -0.0004 -0.0001 89.9999>
  }]
 dyn_skirt_middle_base_r [{shapeType=Capsule, length=0, radius=4, shapeName=dyn_skirt_middle_base_r_capsule, rotation=<rot 0 0.0003 0>}]
 dyn_skirt_side_1_base_l [{shapeType=Capsule, length=0, radius=4, shapeName=dyn_skirt_side_1_base_l_capsule}]
 dyn_skirt_side_1_l [{
   shapeType=Capsule
   length=18
   radius=4.5
   shapeName=dyn_skirt_side_1_l_capsule
   center=<13.5271 -0.8969 -1.2778>
   rotation=<rot 0 -0.0018 89.9997>
  }]
 dyn_skirt_side_1_r [
  {
   shapeType=Capsule
   length=14
   radius=5.5
   shapeName=dyn_skirt_side_1_r_capsule
   center=<-15.6578 1.2549 3.344>
   rotation=<rot 0 -0.0004 -80.0001>
  }
  {shapeType=Capsule, length=8, radius=6, shapeName=dyn_skirt_side_1_r_capsule, center=<-15.2153 -8.3417 -0.9398>, rotation=<rot 0 -0.0001 50>}
 ]
 dyn_skirt_side_2_l [{
   shapeType=Capsule
   length=19.5
   radius=5.5
   shapeName=dyn_skirt_side_2_l_capsule
   center=<10.208 -0.519 -1.6786>
   rotation=<rot -3.4048 19.4065 99.7195>
  }]
 dyn_skirt_side_2_r [
  {
   shapeType=Capsule
   length=21
   radius=6
   shapeName=dyn_skirt_side_2_r_capsule
   center=<-9.1527 9.7663 -3.6633>
   rotation=<rot 17.4682 -1.1498 -57.0861>
  }
  {
   shapeType=Capsule
   length=15.5
   radius=6.5
   shapeName=dyn_skirt_side_2_r_capsule
   center=<-9.1953 -6.8303 0.9747>
   rotation=<rot 3.3676 14.4151 100.5758>
  }
 ]
 dyn_skirt_side_base_r [{shapeType=Capsule, length=0, radius=4, shapeName=dyn_skirt_side_base_r_capsule}]
 pelvis [{shapeType=Capsule, length=9, radius=9.5, shapeName=pelvis_capsule, center=<-6.0352 -2.0583 0>}]
 thigh_l [{shapeType=Capsule, length=30.5, radius=8, shapeName=thigh_l_capsule, center=<-23.07 -1.1688 2.6777>, rotation=<rot 85 179.9988 179.9997>}]
 thigh_r [{shapeType=Capsule, length=28.5, radius=9.5, shapeName=thigh_r_capsule, center=<23.0888 0.7154 -2.7708>, rotation=<rot 85 179.9985 -175.0004>}]
}
constraints [
 {
  bone1Name=dyn_skirt_middle_base_r
  bone2Name=pelvis
  swing1LimitDegrees=45
  swing1Motion=Locked
  swing2LimitDegrees=45
  swing2Motion=Locked
  twistLimitDegrees=45
  twistMotion=Locked
 }
 {
  bone1Name=dyn_skirt_back_middle_base_r
  bone2Name=pelvis
  swing1LimitDegrees=45
  swing1Motion=Locked
  swing2LimitDegrees=45
  swing2Motion=Locked
  twistLimitDegrees=45
  twistMotion=Locked
 }
 {
  bone1Name=dyn_skirt_side_base_r
  bone2Name=pelvis
  swing1LimitDegrees=45
  swing1Motion=Locked
  swing2LimitDegrees=45
  swing2Motion=Locked
  twistLimitDegrees=45
  twistMotion=Locked
 }
 {
  bone1Name=dyn_skirt_back_base
  bone2Name=pelvis
  swing1LimitDegrees=45
  swing1Motion=Locked
  swing2LimitDegrees=45
  swing2Motion=Locked
  twistLimitDegrees=45
  twistMotion=Locked
 }
 {
  bone1Name=dyn_skirt_front_base
  bone2Name=pelvis
  swing1LimitDegrees=45
  swing1Motion=Locked
  swing2LimitDegrees=45
  swing2Motion=Locked
  twistLimitDegrees=45
  twistMotion=Locked
 }
 {
  bone1Name=dyn_skirt_back_middle_1_base_l
  bone2Name=pelvis
  swing1LimitDegrees=45
  swing1Motion=Locked
  swing2LimitDegrees=45
  swing2Motion=Locked
  twistLimitDegrees=45
  twistMotion=Locked
 }
 {
  bone1Name=dyn_skirt_side_1_base_l
  bone2Name=pelvis
  swing1LimitDegrees=45
  swing1Motion=Locked
  swing2LimitDegrees=45
  swing2Motion=Locked
  twistLimitDegrees=45
  twistMotion=Locked
 }
 {
  bone1Name=dyn_skirt_middle_1_base_l
  bone2Name=pelvis
  swing1LimitDegrees=45
  swing1Motion=Locked
  swing2LimitDegrees=45
  swing2Motion=Locked
  twistLimitDegrees=45
  twistMotion=Locked
 }
 {
  bone1Name=dyn_skirt_back_middle_1_base_l
  bone2Name=dyn_skirt_back_middle_1_l
  swing1LimitDegrees=20
  swing2LimitDegrees=20
  twistLimitDegrees=20
 }
 {bone1Name=dyn_skirt_side_1_base_l, bone2Name=dyn_skirt_side_1_l, swing1LimitDegrees=10, swing2LimitDegrees=25, twistLimitDegrees=10}
 {bone1Name=dyn_skirt_side_2_l, bone2Name=dyn_skirt_side_1_l, swing1LimitDegrees=5, swing2LimitDegrees=15, twistLimitDegrees=5}
 {bone1Name=dyn_skirt_back_middle_2_l, bone2Name=dyn_skirt_back_middle_1_l, swing1LimitDegrees=10, swing2LimitDegrees=20, twistLimitDegrees=20}
 {bone1Name=dyn_skirt_back_base, bone2Name=dyn_skirt_back_1, swing1LimitDegrees=10, swing2LimitDegrees=20, twistLimitDegrees=10}
 {bone1Name=dyn_skirt_back_2, bone2Name=dyn_skirt_back_1, swing1LimitDegrees=10, swing2LimitDegrees=30, twistLimitDegrees=20}
 {bone1Name=dyn_skirt_side_2_r, bone2Name=dyn_skirt_side_1_r, swing1LimitDegrees=10, swing2LimitDegrees=30, twistLimitDegrees=30}
 {
  bone1Name=dyn_skirt_back_middle_base_r
  bone2Name=dyn_skirt_back_middle_1_r
  swing1LimitDegrees=10
  swing2LimitDegrees=10
  twistLimitDegrees=15
 }
 {bone1Name=dyn_skirt_back_middle_2_r, bone2Name=dyn_skirt_back_middle_1_r, swing1LimitDegrees=20, swing2LimitDegrees=60, twistLimitDegrees=20}
 {bone1Name=dyn_skirt_front_1, bone2Name=dyn_skirt_front_base, swing1LimitDegrees=15, swing2LimitDegrees=20, twistLimitDegrees=30}
 {bone1Name=dyn_skirt_front_2, bone2Name=dyn_skirt_front_1, swing1LimitDegrees=10, swing2LimitDegrees=20, twistLimitDegrees=15}
 {bone1Name=dyn_skirt_back_2, bone2Name=dyn_skirt_back_middle_2_r, swing1LimitDegrees=45, swing2LimitDegrees=45, twistLimitDegrees=45}
 {bone1Name=dyn_skirt_side_2_l, bone2Name=dyn_skirt_back_middle_2_l, swing1LimitDegrees=60, swing2LimitDegrees=60, twistLimitDegrees=60}
 {bone1Name=dyn_skirt_middle_1_base_l, bone2Name=dyn_skirt_middle_1_l, swing1LimitDegrees=5, swing2LimitDegrees=20, twistLimitDegrees=15}
 {bone1Name=dyn_skirt_middle_2_l, bone2Name=dyn_skirt_middle_1_l, swing1LimitDegrees=2.5, swing2LimitDegrees=5, twistLimitDegrees=20}
 {bone1Name=dyn_skirt_side_1_r, bone2Name=dyn_skirt_side_base_r, swing1LimitDegrees=15, swing2LimitDegrees=20, twistLimitDegrees=20}
 {bone1Name=dyn_skirt_middle_1_r, bone2Name=dyn_skirt_middle_base_r, swing1LimitDegrees=10, swing2LimitDegrees=20, twistLimitDegrees=10}
 {bone1Name=dyn_skirt_middle_2_r, bone2Name=dyn_skirt_middle_1_r, swing1LimitDegrees=5, swing2LimitDegrees=20, twistLimitDegrees=10}
 {bone1Name=dyn_skirt_middle_2_r, bone2Name=dyn_skirt_side_2_r, swing1LimitDegrees=45, swing2LimitDegrees=45, twistLimitDegrees=45}
 {bone1Name=dyn_skirt_side_2_l, bone2Name=dyn_skirt_middle_2_l, swing1LimitDegrees=45, swing2LimitDegrees=45, twistLimitDegrees=45}
]
