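# Animation Blueprint for the female CoatCheck body cosmetic, adding secondary motion and corrective posing to its skirt bones.
# It copies the parent mesh's pose, then chains additive bone offsets, a deform Control Rig, pelvis-space RigidBody physics and final bone tweaks.
# Each update, prioritized front-end, slide, crouch and skydive-float/parachute checks set alphas, rig inputs and gravity.
# Sprint and front-end flags also toggle post-physics offsets directly.
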AnimBlueprint /BRCosmetics/Characters/Player/Female/Medium/Bodies/F_MED_CoatCheck/Meshes/F_MED_CoatCheck_AnimBP
target_skeleton=&/F_MED_CoatCheck_Skeleton
graphs [
 AnimGraph {nodes [
   n1 = "Transform (Modify) Bone - Bone: dyn_skirt_middle_1_base_l"(
    ComponentPose=Local To Component(LocalPose=Copy Pose From Mesh() {bUseAttachedParent=true})
    Alpha=Get Dyn_middle_l_alfa() -> (Dyn_middle_l_alfa:float)
    Translation="0,0,0"
    Rotation="-20.000,0.000,-0.000"
    Scale="1,1,1"
   ) {
    rotationMode=BMM_Additive
    rotationSpace=BCS_BoneSpace
    translationMode=BMM_Additive
    translationSpace=BCS_BoneSpace
    alphaBoolBlend {blendOption=Linear}
    boneToModify {boneName=dyn_skirt_middle_1_base_l}
    rotation=<rot -20.0001 0 0>
   }
   n2 = "Transform (Modify) Bone - Bone: dyn_skirt_side_base_r"(
    ComponentPose=@n1
    Alpha=Get Dyn_tranform_slide_alpha() -> (Dyn_tranform_slide_alpha:float)
    Translation="0,-3,0"
    Rotation="0,0,0"
    Scale="1,1,1"
   ) {
    rotationMode=BMM_Additive
    rotationSpace=BCS_BoneSpace
    translationMode=BMM_Additive
    translationSpace=BCS_BoneSpace
    alphaBoolBlend {blendOption=Linear}
    alphaScaleBiasClamp {bInterpResult=true}
    boneToModify {boneName=dyn_skirt_side_base_r}
    translation {y=-3}
   }
   n3 = Control Rig!(
    Source=Component To Local(ComponentPose=@n2)
    Alpha
    Is_fly:double=Get Dyn_is_fly() -> (Dyn_is_fly:double)
    Is_crouch:double=Get Dyn_crouch_alpha() -> (Dyn_crouch_alpha:float)
   ) -> (Pose) {
    alphaBoolBlend {blendOption=Linear}
    alphaScaleBiasClamp {bInterpResult=true, interpSpeedDecreasing=1, interpSpeedIncreasing=1}
    controlRigAssetReference {blueprintRigClass=&/F_MED_CoatCheck_DeformRig_new.F_MED_CoatCheck_DeformRig_new_C}
   }
   n4 = RigidBody!(ComponentPose=Local To Component(LocalPose=@n3), Alpha, OverrideWorldGravity:Vector=Get Dyn_gravity() -> (Dyn_gravity:Vector), ExternalForce) -> (Pose) {
    bOverrideWorldGravity=true
    overridePhysicsAsset=&/F_MED_CoatCheck_Physics
    simulationSpace=BaseBoneSpace
    alphaBoolBlend {blendOption=Linear}
    alphaScaleBiasClamp {bInterpResult=true, interpSpeedDecreasing=1, interpSpeedIncreasing=1}
    baseBoneRef {boneName=pelvis}
    componentAppliedLinearAccClamp=<0.5 0.5 0.5>
    componentLinearAccScale=<1 1 1>
    componentLinearVelScale=<1 1 1>
    overrideWorldGravity {z=-980}
    simSpaceSettings {maxAngularAcceleration=5, maxAngularVelocity=1, maxLinearAcceleration=4, maxLinearVelocity=100, velocityScaleZ=0, worldAlpha=0.25}
   }
   n5 = "Transform (Modify) Bone - Bone: dyn_skirt_back_middle_base_r"!(ComponentPose=@n4, bAlphaBoolEnabled:bool=Get bIsSprinting(), Translation="0,0,0", Rotation="0,0,20", Scale="1,1,1") -> (Pose) {
    alphaInputType=Bool
    rotationMode=BMM_Additive
    alphaBoolBlend {blendInTime=1, blendOption=CubicInOut, blendOutTime=1}
    boneToModify {boneName=dyn_skirt_back_middle_base_r}
    rotation {roll=20}
   }
   n6 = "Transform (Modify) Bone - Bone: dyn_skirt_back_middle_1_l"!(ComponentPose=@n5, bAlphaBoolEnabled:bool=True, Translation="0,0,0", Rotation="-20,0,0", Scale="1,1,1") -> (Pose) {
    alphaInputType=Bool
    rotationMode=BMM_Additive
    rotationSpace=BCS_BoneSpace
    alphaBoolBlend {blendInTime=1, blendOption=CubicInOut, blendOutTime=1}
    boneToModify {boneName=dyn_skirt_back_middle_1_l}
    rotation {pitch=-20}
   }
   n7 = Get bIsInFrontEnd()
   n8 = "Transform (Modify) Bone - Bone: dyn_skirt_side_2_r"!(ComponentPose=@n6, bAlphaBoolEnabled:bool=@n7, Translation="0,0,0", Rotation="32,0,0", Scale="1,1,1") -> (Pose) {
    alphaInputType=Bool
    rotationMode=BMM_Additive
    rotationSpace=BCS_BoneSpace
    alphaBoolBlend {blendOption=Linear}
    boneToModify {boneName=dyn_skirt_side_2_r}
    rotation {pitch=32}
   }
   n9 = "Transform (Modify) Bone - Bone: dyn_skirt_side_2_l"!(ComponentPose=@n8, bAlphaBoolEnabled:bool=@n7, Translation="0,0,0", Rotation="-13,0,0", Scale="1,1,1") -> (Pose) {
    alphaInputType=Bool
    rotationMode=BMM_Additive
    alphaBoolBlend {blendOption=Linear}
    boneToModify {boneName=dyn_skirt_side_2_l}
    rotation {pitch=-13}
   }
   Output Pose(Result=Component To Local(ComponentPose=@n9)) {layerGroup=None}
  ]}
 EventGraph {nodes [
   Event BlueprintUpdateAnimation()
   > n1 = Branch(Condition=Get bIsInFrontEnd())
   > Set Dyn_tranform_slide_alpha(Dyn_tranform_slide_alpha:float=1) -> (Output_Get:float)
   > Set Dyn_crouch_alpha(Dyn_crouch_alpha:float=0) -> (Output_Get:float)
   > Set Dyn_is_fly(Dyn_is_fly:double=0) -> (Output_Get:double)
   > Set Dyn_gravity(Dyn_gravity:Vector="0,0,-2000") -> (Output_Get:Vector)
   n2 = Branch(execute=@n1.else, Condition=Get bIsSliding())
   > Set Dyn_tranform_slide_alpha(Dyn_tranform_slide_alpha=1)
   > Set Dyn_crouch_alpha(Dyn_crouch_alpha=0)
   > Set Dyn_is_fly(Dyn_is_fly=0)
   > Set Dyn_gravity(Dyn_gravity="0,0,-980")
   n3 = Branch(execute=@n2.else, Condition=Get bIsCrouching())
   > Set Dyn_tranform_slide_alpha(Dyn_tranform_slide_alpha=0)
   > Set Dyn_crouch_alpha(Dyn_crouch_alpha=1)
   > Set Dyn_is_fly(Dyn_is_fly=0)
   > Set Dyn_gravity(Dyn_gravity="0,0,-980")
   > Set Dyn_middle_l_alfa(Dyn_middle_l_alfa:float=1) -> (Output_Get:float)
   n4 = Branch(execute=@n3.else, Condition=OR Boolean(A=Get bIsSkydiveFloating(), B=Get bIsParachuteOpened()))
   > Set Dyn_tranform_slide_alpha(Dyn_tranform_slide_alpha=0)
   > Set Dyn_crouch_alpha(Dyn_crouch_alpha=0)
   > Set Dyn_is_fly(Dyn_is_fly=1)
   > Set Dyn_gravity(Dyn_gravity="0,0,-120")
   Set Dyn_tranform_slide_alpha(execute=@n4.else, Dyn_tranform_slide_alpha=0)
   > Set Dyn_crouch_alpha(Dyn_crouch_alpha=0)
   > Set Dyn_is_fly(Dyn_is_fly=0)
   > Set Dyn_gravity(Dyn_gravity="0,0,-980")
   > Set Dyn_middle_l_alfa(Dyn_middle_l_alfa=0)
   n5 = Branch(Condition=Get bIsMoving2D())
   Set Dyn_crouch_alpha(execute=@n5.else, Dyn_crouch_alpha=1)
   > Set Dyn_is_fly(Dyn_is_fly=0)
   > Set Dyn_gravity(Dyn_gravity="0,0,-980")
   Branch(Condition=Get bIsParachuteOpened())
   Set Dyn_gravity(Dyn_gravity="0, 0, 0")
   Set Dyn_gravity(Dyn_gravity="0, 0, 0")
   Set Dyn_gravity(Dyn_gravity="0, 0, 0")
   Set Dyn_gravity(Dyn_gravity="0, 0, 0")
   Set Dyn_gravity(Dyn_gravity="0, 0, 0")
  ]}
]
variables [
 Dyn_crouch_alpha {type=number}
 Dyn_gravity {type=Vector, default=<0 0 -980>}
 Dyn_is_fly {type=number}
 Dyn_middle_l_alfa {type=number, default=1}
 Dyn_tranform_slide_alpha {type=number, default=1}
]
