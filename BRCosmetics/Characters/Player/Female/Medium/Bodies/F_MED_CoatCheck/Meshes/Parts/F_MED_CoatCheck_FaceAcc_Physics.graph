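# Physics asset for the face accessory part of the female medium CoatCheck character body, defining simulated left and right hair strands.
# Each side chains a base body and two capsule segments through constraints with limited swing and twist, with looser swing at the outer joint.
# A larger head capsule provides a body alongside the strands, with no constraint linking it to them.
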PhysicsAsset /BRCosmetics/Characters/Player/Female/Medium/Bodies/F_MED_CoatCheck/Meshes/Parts/F_MED_CoatCheck_FaceAcc_Physics
bodies {
 dyn_hair_strend_1_l [{shapeType=Capsule, length=3.5, radius=1.5, shapeName=dyn_hair_strend_1_l_capsule, center=<3.2694 0 0>, rotation=<rot -90 -85.4818 65.4822>}]
 dyn_hair_strend_1_r [{
   shapeType=Capsule
   length=3.5
   radius=1.5
   shapeName=dyn_hair_strend_1_l_capsule
   center=<-3.2694 0 0>
   rotation=<rot -90 -85.4817 -114.5179>
  }]
 dyn_hair_strend_2_l [{
   shapeType=Capsule
   length=3.5
   radius=1.5
   shapeName=dyn_hair_strend_2_l_capsule
   center=<3.2694 0 0>
   rotation=<rot -90 -179.9999 -179.9999>
  }]
 dyn_hair_strend_2_r [{shapeType=Capsule, length=3.5, radius=1.5, shapeName=dyn_hair_strend_2_l_capsule, center=<-3.2694 0 0>, rotation=<rot -90 -179.9999 0.0001>}]
 dyn_hair_strend_base_l [{shapeType=Capsule, length=0, radius=1.5, shapeName=dyn_hair_strend_base_l_capsule}]
 dyn_hair_strend_base_r [{shapeType=Capsule, length=0, radius=1.5, shapeName=dyn_hair_strend_base_r_capsule}]
 head [{shapeType=Capsule, length=2.5, radius=7.8005, center=<5.8994 0.2849 -0.0616>, rotation=<rot 0 43.1845 -89.5256>}]
}
constraints [
 {bone1Name=dyn_hair_strend_1_l, bone2Name=dyn_hair_strend_base_l, swing1LimitDegrees=7.5, swing2LimitDegrees=15, twistLimitDegrees=10}
 {bone1Name=dyn_hair_strend_2_l, bone2Name=dyn_hair_strend_1_l, swing1LimitDegrees=25, swing2LimitDegrees=30, twistLimitDegrees=10}
 {bone1Name=dyn_hair_strend_1_r, bone2Name=dyn_hair_strend_base_r, swing1LimitDegrees=7.5, swing2LimitDegrees=15, twistLimitDegrees=10}
 {bone1Name=dyn_hair_strend_2_r, bone2Name=dyn_hair_strend_1_r, swing1LimitDegrees=25, swing2LimitDegrees=30, twistLimitDegrees=10}
]
