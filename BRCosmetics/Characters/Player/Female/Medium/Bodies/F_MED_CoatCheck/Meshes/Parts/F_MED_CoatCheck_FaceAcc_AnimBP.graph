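# Animation Blueprint for the CoatCheck female character's face accessory part, adding secondary physics motion, including to its ponytail bone.
# It copies the pose from the attached parent mesh, then runs an Anim Dynamics box body on dyn_ponytail at half alpha, constrained by angular limits.
# A RigidBody node follows, using the part's physics asset with overridden gravity and tuned simulation-space settings.
# The event graph holds no logic.
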AnimBlueprint /BRCosmetics/Characters/Player/Female/Medium/Bodies/F_MED_CoatCheck/Meshes/Parts/F_MED_CoatCheck_FaceAcc_AnimBP
target_skeleton=&/F_MED_CoatCheck_FaceAcc_Skeleton
graphs [
 AnimGraph {nodes [
   n1 = Anim Dynamics - Bone: dyn_ponytail(ComponentPose=Local To Component(LocalPose=Copy Pose From Mesh() {bUseAttachedParent=true}), Alpha=0.5, ExternalForce="0,0,0") {
    bUseGravityOverride=true
    alphaBoolBlend {blendOption=Linear}
    boundBone {boneName=dyn_ponytail}
    componentAppliedLinearAccClamp=<50 50 50>
    componentLinearAccScale=<0.5 0.5 0.5>
    componentLinearVelScale=<0.5 0.5 0.5>
    gravityOverride {z=-10}
    physicsBodyDefinitions [{
      boundBone {boneName=dyn_ponytail}
      boxExtents=<3 10 7.5>
      constraintSetup {angularLimitsMax=<5 5 10>, angularLimitsMin=<-5 -20 -10>, angularTarget=<0 0 0>}
      localJointOffset=<5 0 -3>
     }]
    relativeSpaceBone {boneName=head}
   }
   n2 = RigidBody(ComponentPose=@n1) {
    bOverrideWorldGravity=true
    overridePhysicsAsset=&/F_MED_CoatCheck_FaceAcc_Physics
    alphaBoolBlend {blendOption=Linear}
    componentAppliedLinearAccClamp=<300 300 300>
    componentLinearAccScale=<1 1 1>
    componentLinearVelScale=<1 1 1>
    overrideWorldGravity {z=-600}
    simSpaceSettings {maxAngularAcceleration=5, maxAngularVelocity=1, maxLinearAcceleration=4, maxLinearVelocity=100, velocityScaleZ=0, worldAlpha=0.25}
   }
   Output Pose(Result=Component To Local(ComponentPose=@n2)) {layerGroup=None}
  ]}
 EventGraph {nodes [Event BlueprintUpdateAnimation()]}
]
variables []
